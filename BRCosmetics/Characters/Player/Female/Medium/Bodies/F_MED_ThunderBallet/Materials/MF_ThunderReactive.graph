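# Material function for the ThunderBallet character that adds animated, reactive lightning glow to input material attributes' emissive.
# Scrolling, noise-distorted thunder and electric textures over pre-skinned local position and UVs combine with fresnel, masked by LightningArea.
# A timestamp decay, overridden in the front end, drives ReactiveAlpha, blending in triple-sine flashing that also pulses a backpack radial glow.
# Extra outputs expose reactive alpha, flashing and temporal responsiveness; hair branches and time-of-day scaling adjust intensity.
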MaterialFunction /BRCosmetics/Characters/Player/Female/Medium/Bodies/F_MED_ThunderBallet/Materials/MF_ThunderReactive
graph {nodes [
  n1 = GetMaterialAttributes(None=FunctionInput() {inputName=Material, inputType=FunctionInput_MaterialAttributes}) {attributeGetTypes [#1, #2]}
  n2 = ComponentMask(None=NamedRerouteUsage()) {b=true, r=true}
  n3 = ComponentMask(None=VectorParameter() {defaultValue=<rgba 0 -4 0 1>}.RGB) {g=true, r=true}
  n4 = TextureSampleParameter2D() {group=Thunder, parameterName=LightningArea, samplerType=SAMPLERTYPE_Masks, texture=&T_BlackMask}
  n5 = Floor(None=Multiply(A=@n4.B) {constB=3})
  n6 = ScalarParameter() {group=Thunder, parameterName=IsHair}
  n7 = Add(A=Add(A=Time(), B=ScalarParameter() {parameterName=ReactivityOffset}), B=Multiply(A=Divide(A=@n5) {constB=3}, B=OneMinus(None=@n6)))
  n8 = ScalarParameter() {defaultValue=1.5, parameterName=ReactiveSpeed}
  n9 = Sine(None=Multiply(A=Multiply(A=@n7, B=@n8)) {constB=3.6874})
  n10 = Sine(None=Multiply(A=Multiply(A=@n7, B=@n8)) {constB=1.7685})
  n11 = Multiply(A=Multiply(A=@n7, B=@n8)) {constB=0.5684}
  n12 = Multiply(A=Multiply(A=ConstantBiasScale(None=@n9), B=ConstantBiasScale(None=@n10)), B=ConstantBiasScale(None=Sine(None=@n11)))
  n13 = LinearInterpolate(
   A=ScalarParameter() {defaultValue=15, group=Reactive, parameterName=RadiusReactMin}
   B=ScalarParameter() {defaultValue=30, group=Reactive, parameterName=RadiusReactMax}
   Alpha=@n12
  )
  n14 = Saturate(None=Power(Base=SphereMask(A=@n2, B=@n3, Radius=@n13) {hardnessPercent=0}, Exp=Constant() {r=2}))
  n15 = VectorParameter() {group=Reactive, parameterName=OuterBlue, defaultValue=<rgba 0 0.3948 1 1>}
  n16 = VectorParameter() {group=Reactive, parameterName=InnerWhite, defaultValue=<rgba 0.5 0.6974 1 1>}
  n17 = LinearInterpolate(A=Multiply(A=@n15.RGB, B=@n15.RGB), B=Multiply(A=@n16.RGB, B=@n16.RGB), Alpha=@n14)
  n18 = Multiply(A=Multiply(A=Multiply(A=@n14, B=@n17), B=NamedRerouteUsage()), B=ScalarParameter() {parameterName=IsBackpack})
  n19 = FunctionInput(Preview=VertexInterpolator(VS=PreSkinnedPosition()).PS) {bUsePreviewValueAsDefault=true, inputName=LocalPOS, sortPriority=1}
  n20 = ComponentMask(None=@n19) {b=true, r=true}
  n21 = Divide(A=@n20, B=ScalarParameter() {defaultValue=40, group=Thunder, parameterName=LightningUnit})
  n22 = VectorParameter() {group=Thunder, parameterName=LightSpeed, defaultValue=<rgba 0 -0.1 0 -2>}
  n23 = Add(A=@n21, B=Multiply(A=ComponentMask(None=@n22.RGB) {g=true, r=true}, B=Time()))
  n24 = NamedRerouteUsage()
  n25 = ScalarParameter() {defaultValue=0.5, parameterName=DistortOverride}
  n26 = VectorParameter() {group=Thunder, parameterName=DistortSpeed, defaultValue=<rgba 0 -0.23 -1 0>}
  n27 = TextureSample(UVs=Add(A=Multiply(A=@n24, B=@n25), B=ComponentMask(None=@n26.RGB) {g=true, r=true})) {samplerSource=SSM_Wrap_WorldGroupSettings, samplerType=SAMPLERTYPE_Normal, texture=/Game/Global/Textures/Default/Noise/T_TilingClouds_N}
  n28 = Multiply(
   A=NamedRerouteDeclaration(None=ComponentMask(None=@n27.RGB) {g=true, r=true}) {name=SwirlNoise, nodeColor=<rgba 1 0 0.2812 1>}
   B=Constant() {r=0.05}
  ) {constB=0.1}
  n29 = TextureSample(UVs=Add(A=NamedRerouteUsage(), B=Add(A=@n23, B=@n28))) {
   samplerSource=SSM_Wrap_WorldGroupSettings
   texture=/BRCosmetics/Characters/Player/Female/Medium/Bodies/F_MED_ThunderBallet/Textures/T_Thunder
  }
  n30 = Multiply(A=Time(), B=ScalarParameter() {defaultValue=1, group=Thunder, parameterName=FlashSpeed}) {constB=3}
  n31 = MaterialFunctionCall(A=@n29.R, B=@n29.R, C=@n29.R, Alpha=ConstantBiasScale(None=Sine(None=@n30))) {materialFunction=&Lerp_3Color}
  n32 = ScalarParameter() {defaultValue=10, parameterName=LightningUnit01}
  n33 = Panner(Coordinate=Add(A=Divide(A=@n20, B=@n32), B=Constant2Vector() {g=0.5, r=0.1})) {speedY=-1}
  n34 = TextureSample(UVs=Add(A=NamedRerouteUsage(), B=Add(A=@n28, B=@n33))) {
   samplerSource=SSM_Wrap_WorldGroupSettings
   texture=&T_Thunder
  }
  n35 = Sine(None=Multiply(A=Time()) {constB=3})
  n36 = Add(
   A=@n31.Result
   B=MaterialFunctionCall(A=@n34.R, B=@n34.R, C=@n34.R, Alpha=ConstantBiasScale(None=@n35)) {materialFunction=&Lerp_3Color}.Result
  )
  n37 = VectorParameter() {group=Thunder, parameterName=ThunderColor, defaultValue=<rgba 0.5344 0.8575 1 100>}
  n38 = Multiply(A=@n37.RGB, B=@n37.RGB)
  n39 = Multiply(
   A=FunctionInput() {bUsePreviewValueAsDefault=true, inputName=AdditionalLightningArea, inputType=FunctionInput_Scalar, sortPriority=10}
   B=NamedRerouteUsage()
  )
  n40 = Multiply(
   A=Multiply(A=NamedRerouteUsage(), B=Multiply(A=@n36, B=@n38))
   B=NamedRerouteDeclaration(None=Add(A=@n39, B=@n4.R)) {name=LightningArea, nodeColor=<rgba 0 1 0.9531 1>}
  )
  n41 = NamedRerouteUsage()
  n42 = NamedRerouteUsage()
  n43 = LinearInterpolate(A=Constant() {r=1}, B=LinearInterpolate(Alpha=@n42) {constA=0.5}, Alpha=ScalarParameter() {group=Thunder, parameterName=IsHair})
  n44 = Add(A=Add(A=@n1.EmissiveColor, B=LinearInterpolate(A=@n40, Alpha=@n6) {constB=0}), B=Multiply(A=Multiply(A=@n41, B=@n43), B=NamedRerouteUsage()))
  n45 = VectorParameter() {group=Thunder, parameterName=FresnelColor, defaultValue=<rgba 0 0 1 2>}
  n46 = Multiply(A=@n45.RGB, B=@n45.RGB)
  n47 = NamedRerouteUsage()
  n48 = Multiply(A=Multiply(A=Fresnel() {baseReflectFraction=0, exponent=2}, B=Multiply(A=@n46, B=@n47)), B=NamedRerouteUsage())
  n49 = LinearInterpolate(
   B=LinearInterpolate(
    A=ScalarParameter() {defaultValue=0.5, parameterName=MinReactiveIntensity}
    B=ScalarParameter() {defaultValue=3, parameterName=ReactiveIntensity}
    Alpha=@n12
   ) {constA=0.5}
   Alpha=NamedRerouteUsage()
  ) {constA=1}
  n50 = Multiply(A=Multiply(A=Add(A=@n44, B=@n48), B=@n49), B=NamedRerouteUsage())
  FunctionOutput(None=SetMaterialAttributes!(MaterialAttributes=@n1.MaterialAttributes, EmissiveColor=Add(A=@n18, B=@n50), BaseColor=@n1.MaterialAttributes) {attributeSetTypes [#1, #2]})
  n51 = Add(A=NamedRerouteUsage(), B=NamedRerouteUsage())
  n52 = Fresnel() {baseReflectFraction=0, exponent=1}
  FunctionOutput(None=Multiply(A=Saturate(None=@n51), B=OneMinus(None=@n52))) {
   description=Connect directly to the Temporal Responsiveness special material output node if this toon character only has material-driven pupils.
   outputName=TemporalResponsiveness
   sortPriority=20
  }
  n53 = Subtract(A=Time(), B=ScalarParameter() {group=Thunder_React, parameterName=Timestamp})
  n54 = ScalarParameter() {defaultValue=6, group=Thunder_React, parameterName=TimePeriod}
  n55 = OneMinus(None=Saturate(None=Divide(A=@n53, B=@n54)))
  n56 = ScalarParameter() {group=Thunder_React, parameterName=ReactiveAlpha}
  n57 = ScalarParameter() {group=Thunder_React, parameterName=IsInFrontEnd}
  FunctionOutput(None=NamedRerouteDeclaration(None=LinearInterpolate(A=@n55, B=@n56, Alpha=@n57)) {name=ReactiveAlpha, nodeColor=<rgba 0 0.5078 1 1>}) {outputName=ReactiveAlpha, sortPriority=10}
  FunctionOutput(None=@n49) {outputName=ReactiveFlashing}
  n58 = LinearInterpolate(Alpha=@n42) {constA=0.1}
  n59 = LinearInterpolate(Alpha=NamedRerouteUsage()) {constA=0.8}
  LinearInterpolate(
   A=@n43
   B=Multiply(A=Multiply(A=@n58, B=@n59), B=ScalarParameter() {defaultValue=3, parameterName=ReactiveIntensity})
   Alpha=NamedRerouteUsage()
  )
  LinearInterpolate(Alpha=NamedRerouteUsage()) {constA=0.5, constB=2}
  Multiply(B=@n43)
  NamedRerouteDeclaration(None=@n31.Result) {name=ElectricMask, nodeColor=<rgba 1 0.2109 0 1>}
  n60 = TextureSample(UVs=Panner(Coordinate=Divide(A=@n20) {constB=40}) {speedY=-0.2}) {samplerSource=SSM_Wrap_WorldGroupSettings, samplerType=SAMPLERTYPE_Masks, texture=/Game/Global/Textures/Default/Noise/T_WrapEdgeWear_Noise}
  n61 = Add(A=Time(), B=@n60.R)
  NamedRerouteDeclaration(None=ConstantBiasScale(None=Sine(None=@n61))) {name=RNoise, nodeColor=<rgba 1 0.6797 0 1>}
  NamedRerouteDeclaration(None=@n19) {name=PreSkin, nodeColor=<rgba 0 0.2969 1 1>}
  n62 = Constant() {r=1}
  n63 = LinearInterpolate(A=Constant() {r=0.8}, Alpha=@n4.R) {constA=1}
  NamedRerouteDeclaration(None=LinearInterpolate(
    B=MaterialFunctionCall(Value_Morning=@n62, Value_Day=@n62, Value_Evening=@n62, Value_Night=@n63) {materialFunction=&Scalar_Time_of_Day}.Result
    Alpha=@n6
   ) {constA=1, constB=0}) {name=TOD, nodeColor=<rgba 0 0.6719 1 1>}
  NamedRerouteDeclaration(None=Constant()) {name=ReactiveFresnelOffset, nodeColor=<rgba 0.7813 0 1 1>}
  NamedRerouteDeclaration(None=@n21) {name=UV01, nodeColor=<rgba 0 1 0.6484 1>}
  n64 = Multiply(A=Time()) {constB=0.5}
  n65 = MaterialFunctionCall(A=@n60.R, B=@n60.R, C=@n60.R, Alpha=ConstantBiasScale(None=Sine(None=@n64))) {materialFunction=&Lerp_3Color}
  n66 = Power(Base=@n65.Result) {constExponent=4}
  n67 = Multiply(A=Time(), B=Constant() {r=0.5}) {constB=0.5}
  n68 = ConstantBiasScale(None=Sine(None=Add(A=@n60.R, B=@n67)))
  n69 = LinearInterpolate(Alpha=@n68) {constA=0.2}
  n70 = VectorParameter() {group=Thunder, parameterName=ElectricTiling, defaultValue=<rgba 20 20 0 1>}
  n71 = BumpOffset(Coordinate=Multiply(A=TextureCoordinate(), B=ComponentMask(None=@n70.RGB) {g=true, r=true}), Height=Constant() {r=-5})
  n72 = Multiply(A=NamedRerouteUsage(), B=ScalarParameter() {defaultValue=0.3, group=Thunder, parameterName=SwirlIntensity01}) {constB=4}
  n73 = TextureSample(UVs=Panner(Coordinate=Add(A=@n71, B=@n72)) {speedY=0.2}) {
   samplerSource=SSM_Wrap_WorldGroupSettings
   texture=/BRCosmetics/Characters/Player/Female/Medium/Bodies/F_MED_ThunderBallet/Textures/T_Electric
  }
  n74 = Multiply(A=Time(), B=ScalarParameter() {defaultValue=2, group=Thunder, parameterName=FlashSpeed01}) {constB=3}
  n75 = NamedRerouteDeclaration(None=@n65.Result) {name=CloudNoise, nodeColor=<rgba 1 0 0.6797 1>}
  n76 = Sine(None=Add(A=Add(A=@n74, B=@n75)) {constB=0.2})
  n77 = VectorParameter() {group=Thunder, parameterName=ElectricColor, defaultValue=<rgba 0.1002 0.076 1 20>}
  n78 = Multiply(
   A=MaterialFunctionCall(A=@n73.R, B=@n73.R, C=@n73.R, Alpha=ConstantBiasScale(None=@n76)) {materialFunction=&Lerp_3Color}.Result
   B=Multiply(A=@n77.RGB, B=@n77.RGB)
  )
  n79 = Fresnel() {baseReflectFraction=0.5, exponent=1}
  NamedRerouteDeclaration(None=Multiply(A=Multiply(A=@n66, B=@n69), B=Multiply(A=@n78, B=@n79))) {name=NoiseElectric, nodeColor=<rgba 0.9219 1 0 1>}
  NamedRerouteDeclaration(None=@n4.G) {name=SkinMask, nodeColor=<rgba 0.7578 0 1 1>}
  NamedRerouteDeclaration(None=@n60.G) {name=GNoise, nodeColor=<rgba 1 0.2344 0 1>}
  NamedRerouteDeclaration(None=LinearInterpolate(Alpha=@n68) {constA=0.1}) {name=SineNoise, nodeColor=<rgba 0.2188 0 1 1>}
  Time()
  Time()
  Time()
 ]}
