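# Master material for the reactive Thunder Ballet backpack cosmetic, built on material attributes for skeletal meshes with a per-expression shading model.
# The shared 2023 character parent PreFX function feeds the outfit's MF_ThunderReactive function, which adds the reactive effect before the PostFX parent function.
# The reactive function also drives a temporal responsiveness output.
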Material /BRCosmetics/Accessories/FORT_Backpacks/Backpack_F_MED_ThunderBallet/Materials/M_ThunderBallet_Reactive_Backpack
graph {
 properties {
  bAutomaticallySetUsageInEditor=false
  bUseMaterialAttributes=true
  bUsedWithMeshDeformer=true
  bUsedWithMorphTargets=true
  bUsedWithSkeletalMesh=true
  pixelDepthOffsetMode=PDOM_Legacy
  shadingModel=MSM_FromMaterialExpression
  usedShadingModels=Default Lit | Subsurface | Clear Coat | Subsurface Profile | Cloth
 }
 nodes [
  n1 = MaterialFunctionCall() {materialFunction=&MF_CharacterParent_2023_PreFX}
  n2 = MaterialFunctionCall!(Material=@n1.Result, LocalPOS=@n1.Result, AdditionalLightningArea) {materialFunction=/BRCosmetics/Characters/Player/Female/Medium/Bodies/F_MED_ThunderBallet/Materials/MF_ThunderReactive}
  MaterialFunctionCall(Material=@n2.Result, M=@n1.M, Pre Skinned Local Position=@n1.M, Distance Blend=@n1.M, Fresnel=@n1.M) {materialFunction=&MF_CharacterParent_2023_PostFX}
  NamedRerouteDeclaration(None=@n1.Pre skinned Local Position) {name=PreskinnedPOS, nodeColor=<rgba 0 1 0.9297 1>}
  TemporalResponsivenessOutput!(None=@n2.TemporalResponsiveness)
 ]
}
